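; Lock actor for the Sideways chest, a child Blueprint that handles the lock's locked and unlocked appearance.
; SetLockVisuals selects a mesh and material from the inherited lock state and applies them to the lock mesh after validity checks.
; On unlock it destroys the idle effect and beam mesh components.
; The construction script rotates both lock mesh and beams from LockMeshRotation plus a fixed offset.
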(event EventOnUnlocked
  (Components|DestroyComponent (Variables|Default|GetNS_SidewaysChest_Lock_Idle))
  (Components|DestroyComponent (Variables|Default|GetSM_SidewaysChest_Lock_Beams)))

(event EventBeginPlay
  (|CallParentFunction) ; This node is disabled and will not be called.\nDrag off pins to build functionality.
)

(event Collision|EventActorBeginOverlap
  (params (OtherActor :type Actor%20Object%20Reference)))

(event EventTick
  (params (DeltaSeconds :type Float%20%28single-precision%29)))

(fn SetLockVisuals
  (bind _lockstate (Variables|LockParent|GetLockState))
  (bind _unlockedmesh (Variables|LockCosmetics|GetUnlockedMesh))
  (bind _lockedmesh (Variables|LockCosmetics|GetLockedMesh))
  (bind _returnvalue (Utilities|Select :Index _lockstate :NewEnumerator1 _lockedmesh :NewEnumerator2 _unlockedmesh))
  (bind _lockstate_2 (Variables|LockParent|GetLockState))
  (bind _unlockedmaterial (Variables|LockCosmetics|GetUnlockedMaterial))
  (bind _lockedmaterial (Variables|LockCosmetics|GetLockedMaterial))
  (bind _returnvalue_2 (Utilities|Select :Index _lockstate_2 :NewEnumerator1 _lockedmaterial :NewEnumerator2 _unlockedmaterial))
  (Utilities|IsValid _returnvalue
    (:Is%20Valid
      (Components|StaticMesh|SetStaticMesh (Variables|Default|GetLockMesh) _returnvalue))
    (:Is%20Not%20Valid
      _))
  (Utilities|IsValid _returnvalue_2
    (:Is%20Valid
      (Rendering|Material|SetMaterial (Variables|Default|GetLockMesh) 0 _returnvalue_2))))

(fn ConstructionScript
  (bind _lockmeshrotation (Variables|Mesh|GetLockMeshRotation))
  (bind _returnvalue (+ -17.1 _lockmeshrotation))
  (bind _returnvalue_2 (Math|Rotator|MakeRotator _returnvalue 0.0 -1.8e+02))
  (|CallParentFunction)
  (Transformation|SetRelativeRotation (Variables|Default|GetLockMesh) _returnvalue_2 false false)
  (Transformation|SetRelativeRotation (Variables|Default|GetSM_SidewaysChest_Lock_Beams) _returnvalue_2 false false))
