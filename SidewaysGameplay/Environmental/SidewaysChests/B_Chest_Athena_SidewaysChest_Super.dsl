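; Sideways chest Blueprint, built on the tiered chest parent, that fades dynamically spawned chests in and out.
; BeginPlay starts a game-time-stamped material fade on a dynamic instance; the server then flags FadedIn, while clients later restore the original material.
; DestroyDynamicChest reverses the fade, then calls SilentDie after a delay.
; It also clears smoke, aura, light, audio and indicators once on opening, and provides a lock mesh lookup.
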(event EventHideChestOpenableCosmetics
  (Utilities|FlowControl|DoOnce false
    (:Completed
      (Rendering|SetVisibility (Variables|Default|GetCustom%20Smoke%20Sheet) false false)
      (Components|DestroyComponent (Variables|Default|GetCustom%20Chest%20Aura))
      (CallFunction|SetLightVisibility false)
      (Audio|Components|Audio|FadeOut (Variables|Default|GetChest_Ambient_Sound) 0.5 0.0 "Linear" :!class /Script/Engine.AudioComponent)
      (CallFunction|StopAudiovisualIndicator)
      (CallFunction|CleanupWiggleMIDs)
      (Components|Timeline|Stop (Variables|Tiered_Chest_Athena_Parent|GetMobileSelectedTL))
      (CallFunction|SetVisibleMobileInteractIcon false)
      (Rendering|SetRenderCustomDepth (Variables|Debug|GetStaticMeshComponent) false)
      (Components|Activation|Activate (Variables|Default|GetNS_Chest_Opened) true))))

(event EventBeginPlay
  ;[1 1 1] Begin Play - Check if it is a dynamic spawn. If so begin the fade.
  (|CallParentFunction)
  (if (and (Variables|Default|GetDynamicSpawn) (not (Variables|Default|GetFadedIn)))
    (CallFunction|SidewaysFadeBegin (Transformation|GetActorLocation (Variables|Default|GetCoreReference)))))

(event Custom|SidewaysFadeBegin
  (params (CoreLocation :type Vector))
  ;[1 1 1] Sideways Fade Begin - Fade in the mesh of the pillars
  (CallFunction|SetMeshandMatVars)
  (CallFunction|CreateFadeMIDs)
  (CallFunction|FadeMaterial (Variables|Default|GetChest_MID))
  ;;[1 1 1] After A delay set a value to ensure we do not hit a race condition but also so that we never call this from begin play again
  (Utilities|FlowControl|Delay 0.1)
  (if (Networking|IsServer)
    (Variables|Default|SetFadedIn true)
    (else
      (Utilities|FlowControl|Delay (Variables|Default|GetDelayForFadingOut))
      (CallFunction|ApplyMaterials (Variables|Default|GetChestMaterial)))))

(event Custom|SidewaysFadeOut
  (params (CoreLocation :type Vector))
  (bind _chest_mid (Variables|Default|GetChest_MID))
  ;[1 1 1] Sideways Fade Out - Fade out the mesh of the pillars - Note the fade out delay needs to match this fade time (roughly)
  (CallFunction|ApplyMaterials _chest_mid)
  (CallFunction|FadeMaterial _chest_mid 1.0)
  (Development|SetHiddeninGame (Variables|Default|GetCustom%20Smoke%20Sheet) true false))

(event Custom|DestroyDynamicChest
  ;[1 1 1] Destory Dynamic Pillar - Call the fade out then after a delay call Silent Die
  (if (Variables|Default|GetDynamicSpawn)
    (CallFunction|SidewaysFadeOut (Transformation|GetActorLocation (Variables|Default|GetCoreReference)))
    (Utilities|FlowControl|Delay (Variables|Default|GetDelayForFadingOut))
    (Gameplay|SilentDie false)))

(fn ApplyMaterials
  (params (ChestMaterial :type Material%20Interface%20Object%20Reference))
  (Rendering|Material|SetMaterial (Variables|Debug|GetStaticMeshComponent) 0 ChestMaterial))

(fn CreateFadeMIDs
  (bind _returnvalue (Rendering|Material|CreateDynamicMaterialInstance (Variables|Debug|GetStaticMeshComponent) 0 (Variables|Default|GetChestFadeMaterial) "None" :!class /Script/Engine.PrimitiveComponent))
  (Variables|Default|SetChest_MID _returnvalue))

(fn FadeMaterial
  (params (MID :type Material%20Instance%20Dynamic%20Object%20Reference) (ReverseFade :type Float%20%28double-precision%29))
  (Rendering|Material|SetScalarParameterValue MID "UseManualFade" 0.0 :!class /Script/Engine.MaterialInstanceDynamic)
  (Rendering|Material|SetScalarParameterValue MID "ReverseFade" ReverseFade :!class /Script/Engine.MaterialInstanceDynamic)
  (Rendering|Material|SetScalarParameterValue MID "BP_FadeTimeBegin" (Utilities|Time|GetGameTimeinSeconds) :!class /Script/Engine.MaterialInstanceDynamic))

(fn GetLockMesh
  (params (Lock :type Child%20Actor%20Component%20Object%20Reference))
  ;[1 1 1] Returns the lock mesh from a child actor that is a sideways chest lock
  (bind _asbga_athena_sideways_chest_lock (Utilities|Casting|CastToBGA_Athena_SidewaysChest_Lock (Variables|/Script/Engine.ChildActorComponent|GetChildActor Lock)))
  (return (Variables|/SidewaysGameplay/Environmental/SidewaysChests/BGA_Athena_SidewaysChest_Lock.BGA_Athena_SidewaysChest_Lock_C|GetLockMesh _asbga_athena_sideways_chest_lock)))

(fn SetMeshAndMatVars
  (Variables|Default|SetChestMaterial (Rendering|Material|GetMaterial (Variables|Debug|GetStaticMeshComponent) 0)))

(fn ConstructionScript
  (|CallParentFunction))
